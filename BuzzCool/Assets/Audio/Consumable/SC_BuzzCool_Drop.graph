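# Sound cue for dropping the BuzzCool consumable, layering a drop whoosh with player finger-wave foley.
# A mixer blends two random branches: one of three whoosh waves at reduced level and one of five foley waves at full level.
# It plays through the gadget items sound class with weapon drop attenuation, and its concurrency override is set to prevent new instances.
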SoundCue /BuzzCool/Assets/Audio/Consumable/SC_BuzzCool_Drop
root=Mixer_1
nodes [
 {id=Mixer_1, children [Random_0,Random_1], properties {inputVolume [0.5,1]}}
 {id=Random_0, children [WavePlayer_0,WavePlayer_1,WavePlayer_2], properties {weights [1,1,1]}}
 {id=Random_1, children [WavePlayer_3,WavePlayer_4,WavePlayer_5,WavePlayer_6,WavePlayer_7], properties {weights [1,1,1,1,1]}}
 {id=WavePlayer_0, properties {soundWaveAssetPtr=/BuzzCool/Assets/Audio/wavs/DropWhoosh_01}}
 {id=WavePlayer_1, properties {soundWaveAssetPtr=/BuzzCool/Assets/Audio/wavs/DropWhoosh_02}}
 {id=WavePlayer_2, properties {soundWaveAssetPtr=/BuzzCool/Assets/Audio/wavs/DropWhoosh_03}}
 {id=WavePlayer_3, properties {soundWaveAssetPtr=/BuzzCool/Assets/Audio/wavs/PlayerFoley_FingerWave_B_01}}
 {id=WavePlayer_4, properties {soundWaveAssetPtr=/BuzzCool/Assets/Audio/wavs/PlayerFoley_FingerWave_B_02}}
 {id=WavePlayer_5, properties {soundWaveAssetPtr=/BuzzCool/Assets/Audio/wavs/PlayerFoley_FingerWave_B_03}}
 {id=WavePlayer_6, properties {soundWaveAssetPtr=/BuzzCool/Assets/Audio/wavs/PlayerFoley_FingerWave_B_04}}
 {id=WavePlayer_7, properties {soundWaveAssetPtr=/BuzzCool/Assets/Audio/wavs/PlayerFoley_FingerWave_B_05}}
]
properties {
 attenuationSettings=&Fort_Weapon_Drop_Att
 maxDistance=3000
 priority=0.5
 soundClassObject=&Fort_Gadgets_Items
 volumeMultiplier=0.5
 concurrencyOverrides {resolutionRule=PreventNew}
}
