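# Sound cue for equipping the BuzzCool consumable.
# A Mixer layers a BuzzCool-specific equip wave, at reduced volume, with a generic consumable equip wave at full volume.
# A root Modulator holds volume steady and leaves pitch modulation at its default range.
# The cue halves overall volume, uses generic player-equip attenuation and a melee weapon sound class, and overrides concurrency to cap instances, stopping the oldest.
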SoundCue /BuzzCool/Assets/Audio/Consumable/SC_BuzzCool_Equip
root=Modulator_0
nodes [
 {id=Mixer_0, children [WavePlayer_0,WavePlayer_1], properties {inputVolume [0.75,1]}}
 {id=Modulator_0, children [Mixer_0], properties {volumeMax=1, volumeMin=1}}
 {id=WavePlayer_0, properties {soundWaveAssetPtr=/BuzzCool/Assets/Audio/wavs/BuzzCool_Equip_01}}
 {id=WavePlayer_1, properties {soundWaveAssetPtr=/BuzzCool/Assets/Audio/wavs/Consumable_Equip_02}}
]
properties {
 attenuationSettings=/Game/Sounds/Attenuation/Fort_Player_Equip_Generic_Att
 bOverrideConcurrency=true
 maxDistance=1800
 soundClassObject=&Fort_WeaponMelee_PSM
 volumeMultiplier=0.5
 concurrencyOverrides {maxCount=2, resolutionRule=StopOldest}
}
