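# Sound cue for the BuzzCool consumable landing on the ground, using weapon-drop attenuation and the loot-collect sound class.
# A mixer layers a randomly picked generic consumable drop, passed through a modulator for variation, with a much quieter random toy burger impact.
# A concurrency override allows one instance at a time and rejects new plays while it is active.
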SoundCue /BuzzCool/Assets/Audio/Consumable/SC_BuzzCool_Ground
root=Mixer_0
nodes [
 {id=Mixer_0, children [Modulator_0,Random_1], properties {inputVolume [1,0.15]}}
 {id=Modulator_0, children [Random_0]}
 {id=Random_0, children [WavePlayer_7,WavePlayer_8], properties {weights [1,1]}}
 {id=Random_1, children [WavePlayer_5,WavePlayer_6], properties {weights [1,1]}}
 {id=WavePlayer_5, properties {soundWaveAssetPtr=/BuzzCool/Assets/Audio/wavs/Toy_Burger_Impact_01}}
 {id=WavePlayer_6, properties {soundWaveAssetPtr=/BuzzCool/Assets/Audio/wavs/Toy_Burger_Impact_03}}
 {id=WavePlayer_7, properties {soundWaveAssetPtr=/BuzzCool/Assets/Audio/wavs/Item_Consumables_Generic_Drop_02}}
 {id=WavePlayer_8, properties {soundWaveAssetPtr=/BuzzCool/Assets/Audio/wavs/Item_Consumables_Generic_Drop_01}}
]
properties {
 attenuationSettings=&Fort_Weapon_Drop_Att
 bOverrideConcurrency=true
 maxDistance=3000
 soundClassObject=&Fort_Loot_Collect
 volumeMultiplier=0.6
 concurrencyOverrides {maxCount=1, resolutionRule=PreventNew}
}
