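# Sound cue for the in-air sound of the thrown BuzzCool consumable, played in the Fort weapons sound class.
# A mixer layers a quieter looping BuzzCool throw loop with a full-volume throw one-shot.
# The one-shot is picked at random from three equally weighted medium held-object throw waves.
# Gadget-activate attenuation shapes the spatial falloff, and the cue's overall volume is reduced.
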SoundCue /BuzzCool/Assets/Audio/Consumable/SC_BuzzCool_InAir
root=Mixer_0
nodes [
 {id=Mixer_0, children [WavePlayer_5,Random_0], properties {inputVolume [0.2,1]}}
 {id=Random_0, children [WavePlayer_1,WavePlayer_2,WavePlayer_3], properties {weights [1,1,1]}}
 {id=WavePlayer_1, properties {soundWaveAssetPtr=/BuzzCool/Assets/Audio/wavs/HeldObjects_Medium_Throw_01}}
 {id=WavePlayer_2, properties {soundWaveAssetPtr=/BuzzCool/Assets/Audio/wavs/HeldObjects_Medium_Throw_02}}
 {id=WavePlayer_3, properties {soundWaveAssetPtr=/BuzzCool/Assets/Audio/wavs/HeldObjects_Medium_Throw_03}}
 {id=WavePlayer_5, properties {bLooping=true, soundWaveAssetPtr=/BuzzCool/Assets/Audio/wavs/BuzzCool_ThrowLoop_01}}
]
properties {
 attenuationSettings=&Fort_GadgetActivate_Att
 maxDistance=4000
 soundClassObject=/Game/Sounds/FortSoundClassesAndModes/Fort_Weapons
 volumeMultiplier=0.4
}
