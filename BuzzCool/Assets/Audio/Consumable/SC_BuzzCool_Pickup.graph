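# Sound cue for picking up the BuzzCool consumable, routed to the loot-collect sound class with weapon-pickup attenuation.
# A mixer layers two randomized finger-wave foley variations, one from each of two wave sets, with a donut-pickup stinger mixed slightly lower.
# Concurrency overrides prevent new instances from starting when the limit is reached.
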SoundCue /BuzzCool/Assets/Audio/Consumable/SC_BuzzCool_Pickup
root=Mixer_0
nodes [
 {id=Mixer_0, children [Random_3,Random_4,WavePlayer_0], properties {inputVolume [1,1,0.8]}}
 {id=Random_3, children [WavePlayer_15, WavePlayer_16, WavePlayer_17, WavePlayer_18, WavePlayer_19], properties {weights [1,1,1,1,1]}}
 {id=Random_4, children [WavePlayer_9, WavePlayer_10, WavePlayer_14], properties {weights [1,1,1]}}
 {id=WavePlayer_0, properties {soundWaveAssetPtr=/BuzzCool/Assets/Audio/wavs/MX_S38_IP_BuzzCool_DonutPickup_STNG}}
 {id=WavePlayer_10, properties {soundWaveAssetPtr=/BuzzCool/Assets/Audio/wavs/PlayerFoley_FingerWave_02}}
 {id=WavePlayer_14, properties {soundWaveAssetPtr=/BuzzCool/Assets/Audio/wavs/PlayerFoley_FingerWave_05}}
 {id=WavePlayer_15, properties {soundWaveAssetPtr=/BuzzCool/Assets/Audio/wavs/PlayerFoley_FingerWave_B_01}}
 {id=WavePlayer_16, properties {soundWaveAssetPtr=/BuzzCool/Assets/Audio/wavs/PlayerFoley_FingerWave_B_02}}
 {id=WavePlayer_17, properties {soundWaveAssetPtr=/BuzzCool/Assets/Audio/wavs/PlayerFoley_FingerWave_B_03}}
 {id=WavePlayer_18, properties {soundWaveAssetPtr=/BuzzCool/Assets/Audio/wavs/PlayerFoley_FingerWave_B_04}}
 {id=WavePlayer_19, properties {soundWaveAssetPtr=/BuzzCool/Assets/Audio/wavs/PlayerFoley_FingerWave_B_05}}
 {id=WavePlayer_9, properties {soundWaveAssetPtr=/BuzzCool/Assets/Audio/wavs/PlayerFoley_FingerWave_01}}
]
properties {
 attenuationSettings=&Fort_Weapon_Pickup_Att
 maxDistance=3500
 priority=0.5
 soundClassObject=&Fort_Loot_Collect
 volumeMultiplier=1
 concurrencyOverrides {resolutionRule=PreventNew}
}
