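# Sound cue for the BuzzCool consumable pickup sting, playing a temporary placeholder wav through a single wave player at slightly reduced volume.
# It uses the loot-collect sound class, overrides concurrency to cap simultaneous instances at two, and disables virtualization.
# A full-level send feeds the weapon-volume sidechain audio bus.
# The debug flag is left enabled.
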SoundCue /BuzzCool/Assets/Audio/Consumable/SC_BuzzCool_PickupSTING
root=WavePlayer_20
nodes [{id=WavePlayer_20, properties {soundWaveAssetPtr=/BuzzCool/Assets/Audio/TEMPwavs/TEMPSTING_BuzzCool_Pickup}}]
properties {
 bDebug=true
 bOverrideConcurrency=true
 maxDistance=2097152
 soundClassObject=&Fort_Loot_Collect
 virtualizationMode=Disabled
 volumeMultiplier=0.85
 busSends [{audioBus=&AB_Sidechain_FortWeapon_Vol, minSendLevel=0, sendLevel=1, soundSourceBus=null}]
 concurrencyOverrides {maxCount=2}
}
