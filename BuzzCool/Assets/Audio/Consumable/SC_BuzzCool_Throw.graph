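# Sound Cue for throwing the BuzzCool consumable, layering a throw sound with a musical stinger.
# A Mixer combines a modulated random pick of three frag grenade throw waves with a random pick of the DonutThrow or End stinger, slightly boosting the throw layer.
# Three quality levels share that Mixer and the fourth slot is empty.
# It uses basic Foley attenuation, the player Foley sound class and a concurrency override that prevents new instances at the limit.
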SoundCue /BuzzCool/Assets/Audio/Consumable/SC_BuzzCool_Throw
root=QualityLevel_0
nodes [
 {id=Mixer_0, children [Modulator_0,Random_1], properties {inputVolume [1.1,1]}}
 {id=Modulator_0, children [Random_0]}
 {id=QualityLevel_0, children [Mixer_0,Mixer_0,Mixer_0,null]}
 {id=Random_0, children [WavePlayer_0,WavePlayer_1,WavePlayer_2], properties {weights [1,1,1]}}
 {id=Random_1, children [WavePlayer_4,WavePlayer_5], properties {weights [1,1]}}
 {id=WavePlayer_0, properties {soundWaveAssetPtr=/BuzzCool/Assets/Audio/wavs/Grenade_Frag_Throw_01}}
 {id=WavePlayer_1, properties {soundWaveAssetPtr=/BuzzCool/Assets/Audio/wavs/Grenade_Frag_Throw_02}}
 {id=WavePlayer_2, properties {soundWaveAssetPtr=/BuzzCool/Assets/Audio/wavs/Grenade_Frag_Throw_03}}
 {id=WavePlayer_4, properties {soundWaveAssetPtr=/BuzzCool/Assets/Audio/wavs/MX_S38_IP_BuzzCool_DonutThrow_STNG}}
 {id=WavePlayer_5, properties {soundWaveAssetPtr=/BuzzCool/Assets/Audio/wavs/MX_S38_IP_BuzzCool_End_STNG}}
]
properties {
 attenuationSettings=&Foley_Attenuation_Basic_Att
 bOverrideConcurrency=true
 maxDistance=1536
 soundClassObject=&Fort_Player_Foley
 volumeMultiplier=1
 concurrencyOverrides {resolutionRule=PreventNew}
}
