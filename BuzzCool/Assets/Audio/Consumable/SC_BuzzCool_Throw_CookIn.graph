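# Sound cue for the cook-in stage of the BuzzCool consumable's throw, layering hand foley, a whoosh and music.
# A mixer plays one of five finger-wave foley sounds at boosted volume alongside a briefly delayed random throw whoosh and a looping BuzzCool attack music track.
# All layers, music included, share the player foley sound class and attenuation, with a concurrency override that blocks new instances at the limit.
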SoundCue /BuzzCool/Assets/Audio/Consumable/SC_BuzzCool_Throw_CookIn
root=Mixer_0
nodes [
 {id=Delay_0, children [Random_2], properties {delayMax=0.2, delayMin=0.2}}
 {id=Mixer_0, children [Random_3,Delay_0,WavePlayer_2], properties {inputVolume [2,1,1.1]}}
 {id=Random_2, children [WavePlayer_6,WavePlayer_7,WavePlayer_8], properties {weights [1,1,1]}}
 {id=Random_3, children [WavePlayer_9, WavePlayer_10, WavePlayer_11, WavePlayer_12, WavePlayer_13], properties {weights [1,1,1,1,1]}}
 {id=WavePlayer_10, properties {soundWaveAssetPtr=/BuzzCool/Assets/Audio/wavs/PlayerFoley_FingerWave_B_02}}
 {id=WavePlayer_11, properties {soundWaveAssetPtr=/BuzzCool/Assets/Audio/wavs/PlayerFoley_FingerWave_B_03}}
 {id=WavePlayer_12, properties {soundWaveAssetPtr=/BuzzCool/Assets/Audio/wavs/PlayerFoley_FingerWave_B_04}}
 {id=WavePlayer_13, properties {soundWaveAssetPtr=/BuzzCool/Assets/Audio/wavs/PlayerFoley_FingerWave_B_05}}
 {id=WavePlayer_2, properties {bLooping=true, soundWaveAssetPtr=/BuzzCool/Assets/Audio/wavs/MX_S38_IP_BuzzCool_Attack_LP}}
 {id=WavePlayer_6, properties {soundWaveAssetPtr=/BuzzCool/Assets/Audio/wavs/Throw_Whoosh_01}}
 {id=WavePlayer_7, properties {soundWaveAssetPtr=/BuzzCool/Assets/Audio/wavs/Throw_Whoosh_02}}
 {id=WavePlayer_8, properties {soundWaveAssetPtr=/BuzzCool/Assets/Audio/wavs/Throw_Whoosh_03}}
 {id=WavePlayer_9, properties {soundWaveAssetPtr=/BuzzCool/Assets/Audio/wavs/PlayerFoley_FingerWave_B_01}}
]
properties {
 attenuationSettings=&Foley_Attenuation_Basic_Att
 bOverrideConcurrency=true
 maxDistance=1536
 soundClassObject=&Fort_Player_Foley
 volumeMultiplier=1
 concurrencyOverrides {resolutionRule=PreventNew}
}
