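# Sound cue for the cook loop of the BuzzCool throwable consumable, played in the player foley sound class with basic foley attenuation.
# A mixer layers two CookLoop waves, each with random pitch variation, the second pitched lower, started after a short fixed delay and mixed louder.
# A concurrency override prevents new instances once the limit is reached, rather than stopping playing ones.
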SoundCue /BuzzCool/Assets/Audio/Consumable/SC_BuzzCool_Throw_CookLoop
root=Mixer_2
nodes [
 {id=Delay_1, children [Modulator_2], properties {delayMax=0.2, delayMin=0.2}}
 {id=Mixer_2, children [Modulator_1,Delay_1], properties {inputVolume [0.5,1.2]}}
 {id=Modulator_1, children [WavePlayer_3], properties {pitchMax=1.1, pitchMin=0.9}}
 {id=Modulator_2, children [WavePlayer_4], properties {pitchMax=0.95, pitchMin=0.85}}
 {id=WavePlayer_3, properties {soundWaveAssetPtr=/BuzzCool/Assets/Audio/wavs/BuzzCool_CookLoop_01}}
 {id=WavePlayer_4, properties {soundWaveAssetPtr=/BuzzCool/Assets/Audio/wavs/BuzzCool_CookLoop_02}}
]
properties {
 attenuationSettings=&Foley_Attenuation_Basic_Att
 bOverrideConcurrency=true
 maxDistance=1536
 soundClassObject=&Fort_Player_Foley
 volumeMultiplier=1
 concurrencyOverrides {resolutionRule=PreventNew}
}
