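# Sound cue for the BuzzCool consumable's throw "cook loop" sound, second variant by name, classed as player foley.
# A single wave player feeds a modulator that randomizes pitch within a lowered range.
# It uses the basic foley attenuation and overrides concurrency so that new instances are prevented once the limit is reached.
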SoundCue /BuzzCool/Assets/Audio/Consumable/SC_BuzzCool_Throw_CookLoop02
root=Modulator_1
nodes [
 {id=Modulator_1, children [WavePlayer_0], properties {pitchMax=0.95, pitchMin=0.85}}
 {id=WavePlayer_0, properties {soundWaveAssetPtr=/BuzzCool/Assets/Audio/wavs/BuzzCool_CookLoop_02}}
]
properties {
 attenuationSettings=&Foley_Attenuation_Basic_Att
 bOverrideConcurrency=true
 maxDistance=1536
 soundClassObject=&Fort_Player_Foley
 volumeMultiplier=1
 concurrencyOverrides {resolutionRule=PreventNew}
}
