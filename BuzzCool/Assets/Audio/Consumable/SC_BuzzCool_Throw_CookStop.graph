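# Sound cue for the BuzzCool consumable, named for the stop of its throw "cook" phase.
# A Random node picks between two equally weighted ice cream throw-ready waves and is excluded from branch culling.
# It plays at reduced volume in the player foley sound class with basic foley attenuation.
# A concurrency override prevents new instances once the limit is reached.
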SoundCue /BuzzCool/Assets/Audio/Consumable/SC_BuzzCool_Throw_CookStop
root=Random_0
nodes [
 {id=Random_0, children [WavePlayer_0,WavePlayer_1], properties {bShouldExcludeFromBranchCulling=true, weights [1,1]}}
 {id=WavePlayer_0, properties {soundWaveAssetPtr=/BuzzCool/Assets/Audio/wavs/Item_Consumables_IceCream_Throw_Ready_010}}
 {id=WavePlayer_1, properties {soundWaveAssetPtr=/BuzzCool/Assets/Audio/wavs/Item_Consumables_IceCream_Throw_Ready_030}}
]
properties {
 attenuationSettings=&Foley_Attenuation_Basic_Att
 bOverrideConcurrency=true
 maxDistance=1536
 soundClassObject=&Fort_Player_Foley
 volumeMultiplier=0.3
 concurrencyOverrides {resolutionRule=PreventNew}
}
